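# Sound cue that plays the ping sound for a Rift point of interest, routed through the critical in-game UI sound class.
# A single wave player feeds a modulator fixed at unity pitch and volume, so playback has no random variation.
# A shared map-object UI ping attenuation shapes it over a very large range, with a slight overall volume boost.
# A concurrency override allows one instance at a time, stopping the oldest.
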SoundCue /RiftPOICosmetics/Sound/MMObj_RiftPOI_Ping_Cue
root=Modulator_5
nodes [
 {id=Modulator_5, children [WavePlayer_6], properties {pitchMax=1, pitchMin=1, volumeMax=1, volumeMin=1}}
 {id=WavePlayer_6, properties {soundWaveAssetPtr=&/S24_Rift_POI_Ping}}
]
properties {
 attenuationSettings=&/MMObj_UI_Ping_Att
 bOverrideConcurrency=true
 maxDistance=2097152
 soundClassObject=&Fort_UI_InGame_Critical
 volumeMultiplier=1.2
 concurrencyOverrides {maxCount=1, resolutionRule=StopOldest}
}
